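# Niagara idle effect for the FareSpore guitar cosmetic that draws an audio-reactive equalizer from a grid of mesh and sprite chips.
# Chips burst-spawns columns times rows particles, partitioned into columns and tinted by a row gradient.
# A local MultsFromSpectrum module turns the music spectrum, or a fake preview spectrum, plus waveform noise into chip scale and sprite visibility.
# A disabled Values emitter holds a per-column spectrum readout.
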NiagaraSystem /SparksCosmetics/Instrument/Guitar/FareSpore/Effects/NS_Guitar_FareSpore_Idle_Visualizer
dependencies {
 usedDataInterfaces [&NiagaraDataInterfaceColorCurve, &NiagaraDataInterfaceCurve, &NiagaraDataInterfaceArrayFloat]
 usedDynamicInputs [
  &Sine
  &Add_Float
  &MultiplyLinearColor
  &MakeLinearColor
 ]
 usedModules [
  &self:SystemSpawnScript.SystemScriptSource.SystemScriptGraph.NiagaraNodeAssignment_4.SetVariables_#1
  &SystemState
  &self:SystemSpawnScript.SystemScriptSource.SystemScriptGraph.NiagaraNodeAssignment_1.SetVariables_#2
  &self:SystemSpawnScript.SystemScriptSource.SystemScriptGraph.NiagaraNodeAssignment_0.SetVariables_#3
  &self:SystemSpawnScript.SystemScriptSource.SystemScriptGraph.NiagaraNodeAssignment_9.SetVariables_#4
  &Timeline
  &self:SystemSpawnScript.SystemScriptSource.SystemScriptGraph.NiagaraNodeAssignment_7.SetVariables_#5
 ]
}
emitters {
 Chips {
  emitterUpdateScript [
   &EmitterState()
   &SpawnBurst_Instantaneous(
    Spawn Count=&Multiply_Int(A=$System.Columns, B=$System.Rows)
    Use Spawn Probability=false
    UseLoopCountLimit=true
   )
  ]
  particleSpawnScript [
   &PartitionParticles(
    Normalized Index In Partition Includes One=true
    Number Of Partitions=$System.Columns
    Particles Per Partition:NiagaraInt32=$System.Rows
    Partition Mode:ENiagara_PartitionMode_v2=::NewEnumerator1
   )
   &InitializeParticle(
    Color Mode=::NewEnumerator0
    Mesh Scale=&Multiply_VectorByFloat(Vector=<0.1 0.02 0.015>, Float=0.28)
    Mesh Scale Mode=::NewEnumerator4
    Position Offset=&Add_Vector(
     A=<8.6 -2.5 -8.25>
     B=&Add_Vector(
      A=&Lerp_Vector(
       A=<0 0 -11.75>
       B=<0 0 11.75>
       Alpha=&Multiply_Float(
        A=$Particles.PartitionParticles.NormalizedPartition
        B=&Divide_Float(
         A=&MakeFloatFromInt(Int=$System.Columns)
         B=&MakeFloatFromInt(Int=&Add_Integer(A=$System.Columns, B=-1))
        )
       )
      )
      B=&Lerp_Vector(A=<0 0 0>, B=<14 0 0>, Alpha=$Particles.PartitionParticles.Normalized Index In Partition)
     )
    )
    Sprite Rotation Angle=90
    Sprite Rotation Mode=::NewEnumerator0
    Sprite Size Mode=::NewEnumerator3
    Uniform Sprite Size=5
    UsePositionOffset=true
    Write Material Random=true
   )
  ]
  particleUpdateScript [
   &ParticleState(Kill Particles When Lifetime Has Elapsed=false, Let Infinitely Lived Particles Die When Emitter Deactivates=true)
   /SetVariables(Transient.NoiseCurve:NiagaraFloat=&Waveform(
     Waveform Count=::NewEnumerator1
     "Waveform [1]"=::NewEnumerator0
     "[1] Phase Scale"=&Multiply_FloatByInt(Float=0.2, Integer=$Particles.PartitionParticles.Partition)
     "[1] Frequency"=1.7
     "[1] Amplitude Min/Max"=<0.25 1>
     "[2] Average"=false
     "[2] Input Position"=0
     "[2] Phase"=$Emitter.Age
     "[2] Phase Scale"=&Multiply_FloatByInt(
      Float=0.2
      Integer=/Niagara/DynamicInputs/Subtract/Subtract_Integer(A:NiagaraInt32=$System.Columns, B:NiagaraInt32=$Particles.PartitionParticles.Partition)
     )
     "[2] Compound Frequency"=2
     "[2] Frequency"=1.5
     "[2] Amplitude Min/Max"=<0.25 1>
     "[2] Scale"=1
     "[2] Bias"=0
     "[2] Exponent"=1
     "[2] Pulse Duration"=0.1
     "[2] Clip Triangle Min/Max"=<0 0>
     "[2] Random Seed"=0
    )) {isSetParameters=true}
   /SetVariables(Transient.EQColorCurve:LinearColor=&ColorFromCurve(
     ColorCurve=DataInterface() {propertyValues {
       AlphaCurve {keys [<at 0 1 RCIM_Cubic>]}
       BlueCurve {keys [<at 0.05 1>,<at 0.3456 1 RCIM_Cubic>,<at 0.8091 1 RCIM_Cubic>]}
       GreenCurve {keys [<at 0.05 0.6344>, {arriveTangent=-0.8357, interpMode=RCIM_Cubic, leaveTangent=-0.8357, time=0.3456, value=0.8218}, <at 0.8091 0 RCIM_Cubic>]}
       RedCurve {keys [<at 0.05 0.04>, {arriveTangent=0.8087, interpMode=RCIM_Cubic, leaveTangent=0.8087, time=0.3456, value=0.44}, <at 0.8091 0.6539 RCIM_Cubic>]}
      }}
     CurveIndex=$Particles.PartitionParticles.Normalized Index In Partition
    )) {isSetParameters=true}
   &self:MultsFromSpectrum(
    Column Selector:NiagaraInt32=$Particles.PartitionParticles.Partition
    Modifier:NiagaraFloat=&Lerp_Float(
     A=&Multiply_Float(
      A=2
      B=&FloatFromCurve(
       FloatCurve=DataInterface() {propertyValues {Curve {keys [<at 0 0.75>,<at 0.25 1>,<at 1 0.5>]}}}
       CurveIndex=&Multiply_FloatByInt(Float=0.125, Integer=$Particles.PartitionParticles.Partition)
       Scale Curve=$NPC.NPC_Audio_Reactive_Parameters.Amplitude_Music
      )
     )
     B=$Transient.NoiseCurve
     Alpha=&Lerp_Float(A=0.4, B=1, Alpha=$System.UsePreview)
    )
    Power Spectrum=2
    Row Selector:NiagaraFloat=$Particles.PartitionParticles.Normalized Index In Partition
    Spectrum Param:LinearColor=&LerpLinearColors(StartColor=$NPC.NPC_Audio_Reactive_Parameters.Spectrum_Music, EndColor=$System.FakeSpectrum, LerpFactor=$System.UsePreview)
   ) {module=MultsFromSpectrum}
   &ScaleColor(
    Scale Alpha=&FloatFromCurve(
     FloatCurve=DataInterface() {propertyValues {Curve {keys [<at 0 1>,<at 1 0.5>]}}}
     CurveIndex=$Particles.PartitionParticles.Normalized Index In Partition
    )
    Scale RGB=&MakeVectorFromLinearColorRGB(Color=$Transient.EQColorCurve)
    ScaleA=true
    ScaleRGB=true
   )
   &ScaleMeshSize(Scale Factor=&VectorFromFloat(Value=$Output.MultsFromSpectrum.SoftMult))
   &ScaleSpriteSize(
    Scale Sprite Size Mode=::NewEnumerator0
    Uniform Scale Factor=&MakeCustomFloatFromBool(
     Bool=&SetBoolByFloatComparison(A=$Output.MultsFromSpectrum.ModifiedMult, B=$Particles.PartitionParticles.Normalized Index In Partition, Comparison Type=::NewEnumerator6)
     True Float=0
     False Float=1
    )
   )
   /SetVariables(
    Particles.SpriteColor=&MakeLinearColorFromVectorAndFloat(
     "Vector (RGB)"=&MakeVectorFromLinearColorRGB(Color=$Particles.Color)
     "Float (Alpha)"=&FloatFromCurve(
      FloatCurve=DataInterface() {propertyValues {Curve {keys [<at 0 0.15>,<at 1 0.4>]}}}
      CurveIndex=$Particles.PartitionParticles.Normalized Index In Partition
      Scale Curve=&Multiply_Float(
       A=&MakeCustomFloatFromBool(Bool=$User.bIsFrontEndPreview, True Float=0.7, False Float=0.65)
       B=&MakeCustomFloatFromBool(
        Bool=&SetBoolByFloatComparison(A=$Output.MultsFromSpectrum.ModifiedMult, B=$Particles.PartitionParticles.Normalized Index In Partition, Comparison Type=::NewEnumerator6)
        True Float=0
        False Float=1
       )
      )
     )
    )
    Particles.SpritePosition:NiagaraPosition=&AddVectorToPosition(Position=$Particles.Position, Vector=<1 0 0>)
   ) {isSetParameters=true}
   /SetVariables(Particles.MeshColor:LinearColor=&Multiply_LinearColorByFloat(Linear Color=$Particles.Color, Float=1.5)) {isSetParameters=true}
  ]
  rendererClasses [&NiagaraSpriteRendererProperties, &NiagaraMeshRendererProperties]
  renderers [
   {rendererClass=&NiagaraSpriteRendererProperties}
   {rendererClass=&NiagaraMeshRendererProperties, rendererIndex=1}
  ]
 }
 Values {
  bEnabled=false
  emitterUpdateScript [
   &EmitterState()
   &SpawnBurst_Instantaneous(Spawn Count=$System.Columns, Use Spawn Probability=false, UseLoopCountLimit=true)
  ]
  particleSpawnScript [
   &PartitionParticles(
    Normalized Index In Partition Includes One=true
    Number Of Partitions=$System.Columns
    Particles Per Partition=1
    Partition Mode:ENiagara_PartitionMode_v2=::NewEnumerator1
   )
   &InitializeParticle(
    Color=<rgba 0 0 0 1>
    Mesh Scale=&Multiply_VectorByFloat(Vector=<0.03 0.03 0.1>, Float=0.28)
    Mesh Scale Mode=::NewEnumerator4
    Position Offset=&Add_Vector(
     A=<26 -2.5 -9.75>
     B=&Lerp_Vector(
      A=<0 0 -11.15>
      B=<0 0 11.15>
      Alpha=&Multiply_Float(A=$Particles.PartitionParticles.NormalizedPartition, B=1.1667)
     )
    )
    Sprite Size Mode=::NewEnumerator3
    UsePositionOffset=true
    Write Material Random=true
   )
  ]
  particleUpdateScript [
   &ParticleState(Kill Particles When Lifetime Has Elapsed=false)
   &self:MultsFromSpectrum(
    Column Selector:NiagaraInt32=$Particles.PartitionParticles.Partition
    Spectrum Param:LinearColor=&LerpLinearColors(StartColor=$NPC.NPC_Audio_Reactive_Parameters.Spectrum_Music, EndColor=$System.FakeSpectrum, LerpFactor=$System.UsePreview)
   ) {module=MultsFromSpectrum}
   &DynamicMaterialParameters(
    Index 0 Param 1=$NPC.NPC_Audio_Reactive_Parameters.Amplitude_Music
    Module.Index0Float0Write=true
    Module.Index0Float1Write=true
    Module.Index0Float2Write=true
    Module.Index0Float3Write=true
   )
   &Color(Color=&ColorFromCurve(
     ColorCurve=DataInterface() {propertyValues {
       AlphaCurve {keys [<at 0 1 RCIM_Cubic>]}
       BlueCurve {keys [<at 0 0 RCIM_Cubic>,<at 1 1 RCIM_Cubic>]}
       GreenCurve {keys [<at 0 0 RCIM_Cubic>,<at 1 0 RCIM_Cubic>]}
       RedCurve {keys [<at 0 1 RCIM_Cubic>,<at 1 0 RCIM_Cubic>]}
      }}
     CurveIndex=$System.Timeline.Float
    ))
   &Color(Color=&MakeCustomLinearColorFromBool(Bool=$User.bIsFrontEndPreview, True Color=<rgba 1 0 0 1>, False Color=<rgba 0 0 1 1>)) {module=Color001}
  ]
  rendererClasses [&NiagaraSpriteRendererProperties]
  renderers [{rendererClass=&NiagaraSpriteRendererProperties}]
 }
}
summary {
 systemName=NS_Guitar_FareSpore_Idle_Visualizer
 emitters [
  {emitterName=Chips, rendererClasses [&NiagaraSpriteRendererProperties, &NiagaraMeshRendererProperties]}
  {bEnabled=false, emitterName=Values, rendererClasses [&NiagaraSpriteRendererProperties]}
 ]
 userVariables [User.bIsFrontEndPreview {type=NiagaraBool}]
}
