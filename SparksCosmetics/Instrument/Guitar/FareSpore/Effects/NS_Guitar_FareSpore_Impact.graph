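# Niagara impact effect for the FareSpore guitar cosmetic: a short burst of glow with light, mesh flare, smoke, sparks and debris.
# Mesh_Box arranges colored meshes and sprites on an expanding torus ring with noise-driven scale, colored by normalized spawn group.
# FN_PhysicalSurface uses User.PhysicalSurface to drive debris velocity, cone angle and gravity.
# Front-end user parameters and time-of-day scalars adjust size, brightness and ring orientation.
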NiagaraSystem /SparksCosmetics/Instrument/Guitar/FareSpore/Effects/NS_Guitar_FareSpore_Impact
dependencies {
 usedDataInterfaces [&NiagaraDataInterfaceCurve, &NiagaraDataInterfaceColorCurve, &NiagaraDataInterfaceCamera]
 usedDynamicInputs [&LerpLinearColors]
 usedModules [
  &self:SystemSpawnScript.SystemScriptSource.SystemScriptGraph.NiagaraNodeAssignment_0.SetVariables_#1
  &self:SystemSpawnScript.SystemScriptSource.SystemScriptGraph.NiagaraNodeAssignment_1.SetVariables_#2
  &SystemState
 ]
}
emitters {
 FN_PhysMatDebris {
  emitterUpdateScript [
   &EmitterState()
   &SpawnBurst_Instantaneous(
    Spawn Count=&RandomRangeInt(Minimum=5, Maximum=10)
    Use Spawn Probability=false
    UseLoopCountLimit=false
   )
  ]
  particleSpawnScript [
   /SetVariables(Particles.BiasedRandom=&FloatFromCurve(
     FloatCurve=DataInterface() {propertyValues {Curve {keys [
         <at 0.2 1 RCIM_Cubic>
         {arriveTangent=-3.3394, interpMode=RCIM_Cubic, leaveTangent=-3.3394, tangentMode=RCTM_User, time=0.5, value=0.5}
         <at 0.8 0 RCIM_Cubic>
        ]}}}
     CurveIndex=&RandomRangeFloat(Minimum=0, Maximum=1)
    )) {isSetParameters=true}
   &InitializeParticle(
    Color Mode=::NewEnumerator0
    Lifetime=&RandomRangeFloat(Minimum=0.5, Maximum=1.25)
    Mass=&Lerp_Float(A=0.6, B=1.2, Alpha=$Particles.BiasedRandom)
    Mass Mode=::NewEnumerator1
    Mesh Scale=&Multiply_VectorByFloat(
     Vector=&RandomRangeVector(Minimum=<0.9 0.4 0.5>, Maximum=<1 1 1>)
     Float=&Lerp_Float(A=0.7, B=1, Alpha=$Particles.BiasedRandom)
    )
    Mesh Scale Mode=::NewEnumerator4
    UsePositionOffset=false
    Write Material Random=true
   )
   &FN_PhysicalSurface(
    Dirt Mesh Count=2
    Glass Mesh Count=3
    Ice Mesh Count=2
    Metal Gravity=&RandomRangeFloat(Minimum=-350, Maximum=-1800)
    Metal Mesh Count=2
    Metal Velocity=&RandomRangeFloat(Minimum=80, Maximum=375)
    Metal Velocity Cone Angle=130
    Physical Surface=$User.PhysicalSurface
    Stone Gravity=&RandomRangeFloat(Minimum=-700, Maximum=-1250)
    Stone Mesh Count=2
    Stone Velocity=&RandomRangeFloat(Minimum=80, Maximum=550)
    Stone Velocity Cone Angle=120
    Wood Gravity=&RandomRangeFloat(Minimum=-450, Maximum=-900)
    Wood Mesh Count=2
    Wood Scale=1.5
    Wood Velocity=&RandomRangeFloat(Minimum=80, Maximum=600)
    Wood Velocity Cone Angle=130
   )
   &AddVelocityInCone(
    Cone Angle=$Output.FN_PhysicalSurface.SurfaceResponseConeAngle
    Module.Use Velocity Falloff On Cone Axis=false
    Velocity Strength=$Output.FN_PhysicalSurface.SurfaceResponseVelocity
   )
   &InitialMeshOrientation(Use Rotation Vector=false)
   &AddRotationalVelocity(Rotation Rate=&RandomVector(Vector Scale=&RandomRangeFloat(Minimum=1, Maximum=2)))
   &ApplyInitialForces()
   /SetVariables(Particles.GravityOverrideValue=$Output.FN_PhysicalSurface.GravityOverride) {isSetParameters=true}
  ]
  particleUpdateScript [
   &ParticleState()
   &GravityForce(Gravity=$Particles.GravityOverrideValue)
   &Collision(
    Advanced Aging Rate=1
    Age Colliding Particles=false
    CPU Collision Trace Channel=ECC_WorldStatic
    Collision Enabled=&Boolean_AndOperation(
     Bool1=&SetBoolByFloatComparison(A=$Particles.Initial.Lifetime, B=1)
     Bool2=&SetBoolByFloatComparison(A=$Particles.Age, B=0.12)
    )
    Control Roll On Collision=true
    Friction=1
    Friction During a Bounce=0.25
    Max CPU Trace Length=100
    Module.Control Rest State Via Surface Normal=false
    Module.EnableMaxCollisionCount=false
    Module.Randomize Collision Normal Vector=true
    Randomize Collision Normal=0.25
    ScaleRateOfRotation=false
   ) {enabled=false}
   &SolveForcesAndVelocity(Clamp Velocity=false, Limit Acceleration=false)
   &ScaleMeshSize(Scale Factor=&VectorFromFloat(Value=&FloatFromCurve(FloatCurve=DataInterface() {propertyValues {Curve {keys [<at 0.6 1 RCIM_Cubic>, {arriveTangent=-6.308, interpMode=RCIM_Cubic, leaveTangent=-6.308, tangentMode=RCTM_User, time=1, value=0}]}}})))
   &DynamicMaterialParameters(Module.Param0WriteEnabled=true, Module.Param1WriteEnabled=true, Module.Param2WriteEnabled=false, Module.Param3WriteEnabled=false)
  ]
  rendererClasses [&NiagaraSpriteRendererProperties, &NiagaraMeshRendererProperties]
  renderers [
   {rendererClass=&NiagaraSpriteRendererProperties}
   {rendererClass=&NiagaraMeshRendererProperties, rendererIndex=1}
   {rendererClass=&NiagaraMeshRendererProperties, rendererIndex=2}
   {rendererClass=&NiagaraMeshRendererProperties, rendererIndex=3}
   {rendererClass=&NiagaraMeshRendererProperties, rendererIndex=4}
   {rendererClass=&NiagaraMeshRendererProperties, rendererIndex=5}
   {rendererClass=&NiagaraMeshRendererProperties, rendererIndex=6}
   {rendererClass=&NiagaraMeshRendererProperties, rendererIndex=7}
   {rendererClass=&NiagaraMeshRendererProperties, rendererIndex=8}
  ]
 }
 Glows {
  emitterUpdateScript [
   &EmitterState()
   &SpawnBurst_Instantaneous(Spawn Count=1, Use Spawn Probability=false)
   &SpawnBurst_Instantaneous(Spawn Count=1, Spawn Group=1, Use Spawn Probability=false) {module=SpawnBurst_Instantaneous001}
  ]
  particleSpawnScript [
   &InitializeParticle(
    Color=$System.TOD
    Lifetime=&MaskFloatBySpawnGroup(Masked Float=0.18, Passthrough Float=0.5)
    Sprite Rotation Mode=::NewEnumerator2
    Sprite Size Mode=::NewEnumerator3
    Uniform Sprite Size=&Multiply_Float(A=175, B=&FN_FrontEndFloatModifier(Gameplay=1, Front End=$User.FrontEndScale))
    UsePositionOffset=false
    Write Material Random=true
   ) {module=InitializeParticle001}
   &DynamicMaterialParameters(
    Index 0 Param 1=1
    Index 0 Param 2=1
    Module.Param0WriteEnabled=true
    Module.Param1WriteEnabled=true
    Module.Param2WriteEnabled=true
    Module.Param3WriteEnabled=true
   )
   &CameraOffset(Camera Offset Amount=&MaskFloatBySpawnGroup(Masked Float=25, Passthrough Float=20))
   /SetVariables(Particles.VisibilityTag=&SpawnGroup()) {isSetParameters=true}
  ]
  particleUpdateScript [
   &ParticleState()
   &ScaleSpriteSize(Uniform Curve Sprite Scale=DataInterface() {propertyValues {Curve {keys [
        {arriveTangent=1.6131, interpMode=RCIM_Cubic, leaveTangent=1.6131, tangentMode=RCTM_User, time=0, value=0.7}
        {arriveTangent=0.0224, interpMode=RCIM_Cubic, leaveTangent=0.0224, tangentMode=RCTM_User, time=1, value=0}
       ]}}})
   &ScaleColor(
    Scale Alpha=&Multiply_Float(
     A=&FN_FrontEndFloatModifier(Gameplay=1, Front End=0.5)
     B=&FNTimeOfDayScalar(Night Scale=0.5, Morning Scale=1, Evening Scale=1)
    )
    Scale RGB=&MakeVectorFromLinearColorRGB(Color=&ColorFromCurve(ColorCurve=DataInterface() {propertyValues {
        AlphaCurve {keys [<at 0 1 RCIM_Cubic>]}
        BlueCurve {keys [<at 0.0251 5>,<at 0.2418 4.9217>]}
        GreenCurve {keys [<at 0.0251 0>,<at 0.2418 5>]}
        RedCurve {keys [<at 0.0251 0.193>,<at 0.2418 0>]}
       }}))
    ScaleA=true
    ScaleRGB=true
   )
   &FN_Light_Attributes(
    Color=$Particles.Color
    Create Color Attribute=true
    Create Exponent Attribute=true
    Create Radius Attribute=true
    Exponent=3
    Radius=90
    Scale Color=3
   )
   /SetVariables(
    Particles.LingerColor:LinearColor=&MultiplyLinearColor(
     Linear Color=<rgba 0 0.75 0.59 1>
     Scale Factor=&MakeFloat4FromVectorAndFloat(
      "Vector (RGB)"=&MakeVectorFromLinearColorRGB(Color=&ColorFromCurve(ColorCurve=DataInterface() {propertyValues {
          AlphaCurve {keys [<at 0 1 RCIM_Cubic>]}
          BlueCurve {keys [{arriveTangent=-1.2161, interpMode=RCIM_Cubic, time=0, value=0.5156}, <at 0.3802 0.3125>, <at 0.9861 0.224>]}
          GreenCurve {keys [{arriveTangent=-3.7684, interpMode=RCIM_Cubic, time=0, value=0.5085}, <at 0.3802 0.1787>, <at 0.9861 0.1281>]}
          RedCurve {keys [{arriveTangent=-5.0588, interpMode=RCIM_Cubic, time=0, value=0.1182}, <at 0.3802 0.001>, <at 0.9861 0.0007>]}
         }}))
      "Float (Alpha)"=&FloatFromCurve(FloatCurve=DataInterface() {propertyValues {Curve {keys [
           {interpMode=RCIM_Cubic, leaveTangentWeight=0.2008, tangentMode=RCTM_User, time=0, value=1}
           {arriveTangent=-1.1533, interpMode=RCIM_Cubic, leaveTangent=-1.1533, tangentMode=RCTM_User, time=0.4, value=0.35}
           {arriveTangent=-0.3667, interpMode=RCIM_Cubic, leaveTangent=-0.3667, tangentMode=RCTM_User, time=1, value=0}
          ]}}})
     )
    )
    Particles.LingerScale:Vector2f=&Vector2DFromFloat(Value=&FloatFromCurve(
      FloatCurve=DataInterface() {propertyValues {Curve {keys [
          {arriveTangent=2.3921, interpMode=RCIM_Cubic, leaveTangent=2.3921, tangentMode=RCTM_User, time=0, value=0.25}
          {interpMode=RCIM_Cubic, tangentMode=RCTM_User, time=0.7, value=1.5}
         ]}}}
      Scale Curve=&Multiply_Float(A=175, B=&FN_FrontEndFloatModifier(Gameplay=1, Front End=$User.FrontEndScale))
     ))
   ) {isSetParameters=true}
  ]
  rendererClasses [&NiagaraSpriteRendererProperties, &NiagaraLightRendererProperties]
  renderers [
   {rendererClass=&NiagaraSpriteRendererProperties}
   {rendererClass=&NiagaraLightRendererProperties, rendererIndex=1}
   {rendererClass=&NiagaraSpriteRendererProperties, rendererIndex=2}
  ]
 }
 Impact_Flare {
  emitterUpdateScript [
   &EmitterState()
   &SpawnBurst_Instantaneous(Spawn Count=1, Use Spawn Probability=false)
  ]
  particleSpawnScript [
   &InitializeParticle(
    Color=$System.TOD
    Lifetime=0.25
    Mesh Scale=&Multiply_VectorByFloat(
     Vector=&RandomRangeVector(Minimum=<1.5 1.5 1.5>, Maximum=<1.7 1.7 1.7>)
     Float=&FN_FrontEndFloatModifier(Gameplay=1, Front End=0.7)
    )
    Mesh Scale Mode=::NewEnumerator4
    UsePositionOffset=false
    Write Material Random=true
   ) {module=InitializeParticle001}
   &InitialMeshOrientation(Use Rotation Vector=false) {module=InitialMeshOrientation001}
  ]
  particleUpdateScript [
   &ParticleState()
   &ScaleColor(
    Scale Alpha=&FloatFromCurve(FloatCurve=DataInterface() {propertyValues {Curve {keys [<at 0 1 RCIM_Cubic>, {arriveTangent=-0.7263, tangentMode=RCTM_User, time=1, value=0}]}}})
    Scale RGB=&MakeVectorFromLinearColorRGB(Color=&Multiply_LinearColorByFloat(
      Linear Color=&ColorFromCurve(ColorCurve=DataInterface() {propertyValues {
         AlphaCurve {keys [<at 0 1 RCIM_Cubic>]}
         BlueCurve {keys [<at 0 1.9872>,<at 0.2603 0.6624>,<at 0.6963 0.5052>,<at 0.9219 0>]}
         GreenCurve {keys [<at 0 3>,<at 0.2603 1>,<at 0.6963 0.163>,<at 0.9219 0>]}
         RedCurve {keys [<at 0 0.2661>,<at 0.2603 0.0887>,<at 0.6963 0.0031>,<at 0.9219 0>]}
        }})
      Float=&Multiply_Float(
       A=&FN_FrontEndFloatModifier(Gameplay=1, Front End=0.7)
       B=&FNTimeOfDayScalar(Night Scale=1.2, Morning Scale=1.1, Evening Scale=1.2)
      )
     ))
    ScaleA=true
    ScaleRGB=true
   )
   &ScaleMeshSize(Scale Factor=&VectorFromFloat(Value=&FloatFromCurve(
      FloatCurve=DataInterface() {propertyValues {Curve {keys [
          {arriveTangent=2.6228, interpMode=RCIM_Cubic, leaveTangent=2.6228, tangentMode=RCTM_User, time=0, value=0.3}
          {arriveTangent=0.0612, interpMode=RCIM_Cubic, leaveTangent=0.0612, tangentMode=RCTM_User, time=0.3, value=0.7}
          {arriveTangent=2.4232, tangentMode=RCTM_User, time=0.7, value=1}
         ]}}}
      Scale Curve=1.25
     )))
   &DynamicMaterialParameters(
    Index 0 Param 1=&FloatFromCurve(FloatCurve=DataInterface() {propertyValues {Curve {keys [<at 0 0 RCIM_Cubic>, {arriveTangent=1.239, tangentMode=RCTM_User, time=1, value=0.75}]}}})
    Index 0 Param 2=45
    Index 0 Param 3=1
    Index 0 Param 4=1
    Module.Param0WriteEnabled=true
    Module.Param1WriteEnabled=true
    Module.Param2WriteEnabled=false
    Module.Param3WriteEnabled=false
   ) {module=DynamicMaterialParameters001}
   &CameraOffset(Camera Offset Amount=20)
  ]
  rendererClasses [&NiagaraMeshRendererProperties]
  renderers [{rendererClass=&NiagaraMeshRendererProperties, rendererIndex=0}]
 }
 Mesh_Box {
  emitterUpdateScript [
   &EmitterState(
    Inactive Response=::NewEnumerator1
    Life Cycle Mode=::NewEnumerator1
    Loop Behavior=::NewEnumerator1
    Loop Duration Mode=::NewEnumerator1
    UseLoopDelay=false
   )
   &SpawnBurst_Instantaneous(Spawn Count=35, Use Spawn Probability=false, UseLoopCountLimit=false)
   &SpawnBurst_Instantaneous(Spawn Count=30, Use Spawn Probability=false, UseLoopCountLimit=false) {enabled=false, module=SpawnBurst_Instantaneous001}
  ]
  particleSpawnScript [
   &FN_SetSpawnGroup(Normalize Spawn Group=true, Particles Per Group=35, Pattern=::NewEnumerator0)
   &InitializeParticle(
    Color Mode=::NewEnumerator0
    Lifetime=0.25
    Mesh Scale=&Multiply_VectorByFloat(
     Vector=<0.7 0.35 0.2>
     Float=&FN_FrontEndFloatModifier(Gameplay=1, Front End=$User.FrontEndScale)
    )
    Mesh Scale Mode=::NewEnumerator4
    Sprite Size Mode=::NewEnumerator3
    Uniform Sprite Size=20
    UsePositionOffset=false
    Write Material Random=true
   )
   &InitialMeshOrientation(Orientation Axis=<0 0 -1>, Use Orientation Vector=true, Use Rotation Vector=true)
   &CameraOffset(Camera Offset Amount=15)
  ]
  particleUpdateScript [
   &SolveForcesAndVelocity(Clamp Velocity=false, Limit Acceleration=false)
   &ParticleState()
   &Color(
    Color=&ColorFromCurve(
     ColorCurve=DataInterface() {propertyValues {
       AlphaCurve {keys [<at 0 1 RCIM_Cubic>]}
       BlueCurve {keys [<at 0.0705 1>,<at 0.1637 1 RCIM_Cubic>,<at 0.3047 1>,<at 0.4153 1>,<at 0.5585 1>,<at 0.7038 1>,<at 0.8252 1>,<at 0.9944 1>]}
       GreenCurve {keys [<at 0.0705 0.6376>,<at 0.1637 0.8228 RCIM_Cubic>,<at 0.3047 0.8228>,<at 0.4153 0.6376>,<at 0.5585 0.6376>,<at 0.7038 0.8228>,<at 0.8252 0.8228>,<at 0.9944 0.6376>]}
       RedCurve {keys [<at 0.0705 0.0395>,<at 0.1637 0.4397 RCIM_Cubic>,<at 0.3047 0.4397>,<at 0.4153 0.0395>,<at 0.5585 0.0395>,<at 0.7038 0.4397>,<at 0.8252 0.4397>,<at 0.9944 0.0395>]}
      }}
     CurveIndex=$Particles.Normalized Spawn Group
    )
    Scale Alpha=0.6
    Scale Color=&VectorFromFloat(Value=&FN_FrontEndFloatModifier(
      Gameplay=&Multiply_Float(A=&FNTimeOfDayScalar(Night Scale=0.6, Morning Scale=0.3, Evening Scale=0.3), B=0.2)
      Front End=0.3
     ))
   ) {enabled=false}
   &Color(
    Color=&ColorFromCurve(
     ColorCurve=DataInterface() {propertyValues {
       AlphaCurve {keys [
         <at 0 0.5>
         {arriveTangent=1.6065, interpMode=RCIM_Cubic, leaveTangent=1.6065, time=0.1225, value=1}
         <at 0.3112 1>
         <at 0.6322 1>
         <at 0.8122 1>
        ]}
       BlueCurve {keys [<at 0.0228 1>,<at 0.0702 1 RCIM_Cubic>,<at 0.1305 1 RCIM_Cubic>,<at 0.2505 1>,<at 0.3112 1>,<at 0.4045 1>,<at 0.4826 1>,<at 0.5281 1>,<at 0.6235 1>,<at 0.6929 1>,<at 0.7363 1>,<at 0.8035 1>,<at 0.9944 1>]}
       GreenCurve {keys [
         <at 0.0228 0.6376>
         {arriveTangent=-4.6615, interpMode=RCIM_Cubic, leaveTangent=-4.6615, time=0.0702, value=0.8228}
         {arriveTangent=-3.8121, interpMode=RCIM_Cubic, leaveTangent=-3.8121, time=0.1305, value=0.1356}
         <at 0.2505 0.1356>
         <at 0.3112 0.8228>
         <at 0.4045 0.8228>
         <at 0.4826 0.6376>
         <at 0.5281 0.6376>
         <at 0.6235 0.8228>
         <at 0.6929 0.1356>
         <at 0.7363 0.1356>
         <at 0.8035 0.8228>
         <at 0.9944 0.6376>
        ]}
       RedCurve {keys [
         <at 0.0228 0.0395>
         {arriveTangent=6.1436, interpMode=RCIM_Cubic, leaveTangent=6.1436, time=0.0702, value=0.4397}
         {arriveTangent=1.4504, interpMode=RCIM_Cubic, leaveTangent=1.4504, time=0.1305, value=0.7011}
         <at 0.2505 0.7011>
         <at 0.3112 0.4397>
         <at 0.4045 0.4397>
         <at 0.4826 0.0395>
         <at 0.5281 0.0395>
         <at 0.6235 0.4397>
         <at 0.6929 0.7011>
         <at 0.7363 0.7011>
         <at 0.8035 0.4397>
         <at 0.9944 0.0395>
        ]}
      }}
     CurveIndex=$Particles.Normalized Spawn Group
    )
    Scale Alpha=0.6
    Scale Color=&VectorFromFloat(Value=&FN_FrontEndFloatModifier(
      Gameplay=&Multiply_Float(A=&FNTimeOfDayScalar(Night Scale=0.6, Morning Scale=0.3, Evening Scale=0.3), B=0.2)
      Front End=0.3
     ))
   ) {module=Color001}
   /SetVariables(Particles.MeshOrientation=&MultiplyQuaternion(
     Quaternion A=&FindQuatBetweenTwoVectors(
      From Vector=<0 0 1>
      To Vector=&NormalizeVector(Vector To Normalize=&Subtract_Vector(
        A=&MakeCustomVectorFromBool(
         Bool=&SetBoolByFloatComparison(A=$Particles.Normalized Spawn Group, B=0.5, Comparison Type=::NewEnumerator4)
         True Vector=&Subtract_Vector(A=Linked() {linkedVariable {name=Particles.Position, type=NiagaraPosition}}, B=<1 0 0>)
         False Vector=Linked() {linkedVariable {name=Particles.Position, type=NiagaraPosition}}
        )
        B=<0 0 0>
       ))
     )
     Quaternion B=$Particles.Initial.MeshOrientation
    )) {isSetParameters=true}
   &self:NoiseScale(
    Noise Frequency=5
    Noise Speed=2
    Scale Size:NiagaraFloat=&FloatFromCurve(
     FloatCurve=DataInterface() {propertyValues {Curve {keys [
         {arriveTangent=0.1523, interpMode=RCIM_Cubic, leaveTangent=0.1523, tangentMode=RCTM_User, time=0, value=0.15}
         {arriveTangent=0.0283, interpMode=RCIM_Cubic, leaveTangent=0.0283, tangentMode=RCTM_User, time=0.7, value=0.5}
        ]}}}
     Scale Curve=&FN_FrontEndFloatModifier(Gameplay=1, Front End=$User.FrontEndScale)
    )
   ) {module=NoiseScale}
   &TorusLocation(
    Align Torus With Axis=true
    Handle Radius=0
    Large Radius:NiagaraFloat=&FloatFromCurve(
     FloatCurve=DataInterface() {propertyValues {Curve {keys [
         {arriveTangent=45.9405, interpMode=RCIM_Cubic, leaveTangent=45.9405, leaveTangentWeight=20.1072, tangentMode=RCTM_User, time=0, value=1}
         {arriveTangent=65.167, arriveTangentWeight=13.336, interpMode=RCIM_Cubic, leaveTangent=65.167, tangentMode=RCTM_User, time=0.8, value=40}
        ]}}}
     Scale Curve=&FN_FrontEndFloatModifier(Gameplay=0.65, Front End=&Multiply_Float(A=0.65, B=0.8))
    )
    MaskSpawnGroup=false
    Normalized Angle Around Torus Axis:NiagaraFloat=$Particles.Normalized Spawn Group
    Torus Axis=&MakeCustomVectorFromBool(
     Bool=$User.bIsFrontEndPreview
     True Vector=&ReturnCameraVector(Camera Vectors=::NewEnumerator1, Coordinate Space=::Simulation)
     False Vector=<1 0 0>
    )
    Torus Distribution Mode:ENiagaraTorusDistributionMode=::NewEnumerator1
    Torus Origin:NiagaraPosition=$Particles.Initial.Position
   )
   &ScaleMeshSize(Scale Factor=&self:XY_to_XYZ(XY:Vector2f=&MakeVector2D(
      X=&FloatFromCurve(FloatCurve=DataInterface() {propertyValues {Curve {keys [
           {interpMode=RCIM_Cubic, tangentMode=RCTM_Break, time=0, value=4}
           {interpMode=RCIM_Cubic, tangentMode=RCTM_Break, time=0.35, value=0.55}
           {arriveTangent=-0.0779, interpMode=RCIM_Cubic, leaveTangent=-0.0779, tangentMode=RCTM_User, time=0.6, value=2.5}
           {interpMode=RCIM_Cubic, tangentMode=RCTM_Break, time=1, value=0}
          ]}}})
      Y=&FloatFromCurve(FloatCurve=DataInterface() {propertyValues {Curve {keys [<at 0 1>,<at 0.6 1>,<at 1 0>]}}})
     ))) {enabled=false}
   &ScaleMeshSize(Scale Factor=&self:XY_to_XYZ(XY:Vector2f=&MakeVector2D(
      X=&FloatFromCurve(FloatCurve=DataInterface() {propertyValues {Curve {keys [
           {arriveTangent=10.5485, interpMode=RCIM_Cubic, leaveTangent=10.5485, tangentMode=RCTM_User, time=0, value=1.25}
           {arriveTangent=0.6667, interpMode=RCIM_Cubic, leaveTangent=0.6667, time=0.2, value=3}
           {arriveTangent=-3.5, interpMode=RCIM_Cubic, leaveTangent=-3.5, time=0.6, value=1.65}
           {arriveTangent=-4.4509, interpMode=RCIM_Cubic, leaveTangent=-4.4509, tangentMode=RCTM_User, time=1, value=0.2}
          ]}}})
      Y=&FloatFromCurve(FloatCurve=DataInterface() {propertyValues {Curve {
          defaultValue=1
          keys [{interpMode=RCIM_Cubic, tangentMode=RCTM_User, time=0, value=0.8}, {interpMode=RCIM_Cubic, tangentMode=RCTM_User, time=1, value=1.2}]
         }}})
     ))) {module=ScaleMeshSize001}
   &ScaleColor(
    Color Value To Scale=$Particles.Color
    Scale Alpha=&FloatFromCurve(
     FloatCurve=DataInterface() {propertyValues {Curve {
        defaultValue=0.475
        keys [<at 0.7 1 RCIM_Cubic>, {arriveTangent=-4.1931, interpMode=RCIM_Cubic, leaveTangent=-4.1931, tangentMode=RCTM_User, time=1, value=0.4}]
       }}}
     Scale Curve=&FN_FrontEndFloatModifier(Gameplay=1, Front End=0.5)
    )
    Scale RGB=&VectorFromFloat(Value=&FloatFromCurve(
      FloatCurve=DataInterface() {propertyValues {Curve {keys [
          {arriveTangent=-0.547, interpMode=RCIM_Cubic, leaveTangent=0.0726, tangentMode=RCTM_Break, time=0.1, value=3}
          {arriveTangent=-4.0081, interpMode=RCIM_Cubic, leaveTangent=-1.7129, tangentMode=RCTM_Break, time=0.5, value=1.2}
          {arriveTangent=-0.547, interpMode=RCIM_Cubic, tangentMode=RCTM_Break, time=1, value=0.9}
         ]}}}
      Scale Curve=&FN_FrontEndFloatModifier(Gameplay=&FNTimeOfDayScalar(Night Scale=1.5, Morning Scale=2, Evening Scale=2), Front End=0.75)
     ))
    ScaleA=true
    ScaleRGB=true
   ) {module=ScaleColor001}
   /SetVariables(
    Particles.SpriteColor=&MakeLinearColorFromVectorAndFloat(
     "Vector (RGB)"=&Multiply_VectorByFloat(Vector=&MakeVectorFromLinearColorRGB(Color=$Particles.Color), Float=1.3)
     "Float (Alpha)"=&FloatFromCurve(FloatCurve=DataInterface() {propertyValues {Curve {keys [<at 0.4 1 RCIM_Cubic>,<at 1 0 RCIM_Cubic>]}}})
    )
    Particles.SpritePosition:NiagaraPosition=&AddVectorToPosition(Position=$Particles.Position, Vector=<1 0 0>)
    Particles.spritesize2:Vector2f=&Vector2DFromFloat(Value=&FloatFromCurve(
      FloatCurve=DataInterface() {propertyValues {Curve {
         defaultValue=1
         keys [{interpMode=RCIM_Cubic, tangentMode=RCTM_User, time=0, value=0}, {interpMode=RCIM_Cubic, tangentMode=RCTM_User, time=0.5, value=1}]
        }}}
      Scale Curve=&Multiply_Float(A=20, B=&FN_FrontEndFloatModifier(Gameplay=1, Front End=$User.FrontEndScale))
     ))
   ) {isSetParameters=true}
   &AlignSpriteToMeshOrientation(SetAlignment=true, SetFacing=true) {enabled=false}
  ]
  rendererClasses [&NiagaraMeshRendererProperties, &NiagaraSpriteRendererProperties]
  renderers [
   {rendererClass=&NiagaraMeshRendererProperties, rendererIndex=0}
   {rendererClass=&NiagaraSpriteRendererProperties, rendererIndex=1}
  ]
 }
 Smoke {
  emitterUpdateScript [
   &EmitterState()
   &SpawnBurst_Instantaneous(Spawn Count=4, Use Spawn Probability=false)
  ]
  particleSpawnScript [
   &InitializeParticle(
    Color=<rgba 0.1615 0.4094 1 0.3>
    Lifetime Max=0.45
    Lifetime Min=0.35
    Lifetime Mode=::NewEnumerator1
    Sprite Rotation Mode=::NewEnumerator2
    Sprite Size Mode=::NewEnumerator1
    Uniform Sprite Size Max=90
    Uniform Sprite Size Min=80
    UsePositionOffset=false
    Write Material Random=true
   ) {module=InitializeParticle001}
   &CameraOffset(Camera Offset Amount=&FN_FrontEndFloatModifier(Gameplay=&UniformRangedFloat(Minimum=20, Maximum=30), Front End=0))
   &ShapeLocation(MaskSpawnGroup=false, Override Seed=false, Sphere Radius=10, UseOwnerScale=true)
   &AddVelocity(Velocity=<50 0 0>)
   &AddVelocity(
    Module.Use Falloff Distance=false
    Override Seed=false
    Velocity Mode=::NewEnumerator1
    Velocity Speed=&RandomRangeFloat(Minimum=10, Maximum=25)
   ) {module=AddVelocity001}
  ]
  particleUpdateScript [
   &ParticleState()
   &GravityForce(Gravity=<0 0 -100>)
   &Drag(Use Linear Drag=true, Use Rotational Drag=true)
   &SolveForcesAndVelocity(Clamp Velocity=false, Limit Acceleration=false)
   /SetVariables(Particles.SubImageIndex=&Lerp_Float(A=0, B=10, Alpha=$Particles.NormalizedAge)) {isSetParameters=true}
   &ScaleSpriteSize(Scale Factor=&Vector2DFromFloat(Value=&FloatFromCurve(
      FloatCurve=DataInterface() {propertyValues {Curve {keys [
          {arriveTangent=3.6818, interpMode=RCIM_Cubic, leaveTangent=3.6818, tangentMode=RCTM_User, time=0, value=0.2}
          {arriveTangent=0.4485, interpMode=RCIM_Cubic, leaveTangent=0.4485, tangentMode=RCTM_User, time=0.3, value=0.85}
          <at 1 1 RCIM_Cubic>
         ]}}}
      Scale Curve=&FN_FrontEndFloatModifier(Gameplay=1, Front End=$User.FrontEndScale)
     )))
   &DynamicMaterialParameters(
    Index 0 Param 1=&FloatFromCurve(FloatCurve=DataInterface() {propertyValues {Curve {keys [<at 0.5 0.5>,<at 1 -0.1>]}}})
    Index 0 Param 4=1
    Module.Index0Float0Write=true
    Module.Index0Float1Write=true
    Module.Index0Float2Write=true
    Module.Index0Float3Write=true
   )
   &ScaleColor(
    Scale Alpha=&FloatFromCurve(
     FloatCurve=DataInterface() {propertyValues {Curve {keys [
         {arriveTangent=3.6206, interpMode=RCIM_Cubic, leaveTangent=3.6206, tangentMode=RCTM_User, time=0, value=0.2}
         {arriveTangent=-0.2, interpMode=RCIM_Cubic, leaveTangent=-0.2, time=0.1, value=0.85}
         <at 1 0 RCIM_Cubic>
        ]}}}
     Scale Curve=&FNTimeOfDayScalar(Night Scale=1, Morning Scale=0.8, Day Scale=0.6, Evening Scale=0.5)
    )
    Scale RGB=&VectorFromFloat(Value=&FloatFromCurve(FloatCurve=DataInterface() {propertyValues {Curve {keys [<at 0.3 1.5>,<at 1 0.8>]}}}))
    ScaleA=true
    ScaleRGB=true
   )
  ]
  rendererClasses [&NiagaraSpriteRendererProperties]
  renderers [{rendererClass=&NiagaraSpriteRendererProperties}]
 }
 Spark {
  emitterUpdateScript [
   &EmitterState()
   &SpawnBurst_Instantaneous(
    Spawn Count=&RandomRangeInt(Minimum=17, Maximum=22)
    Use Spawn Probability=false
    UseLoopCountLimit=false
   )
   &SpawnBurst_Instantaneous(
    Spawn Count=&RandomRangeInt(Minimum=10, Maximum=15)
    Spawn Group=1
    Use Spawn Probability=false
    UseLoopCountLimit=false
   ) {module=SpawnBurst_Instantaneous001}
  ]
  particleSpawnScript [
   &InitializeParticle(
    Lifetime=&RandomRangeFloat(Minimum=0.24, Maximum=0.75)
    Mesh Scale=&Multiply_VectorByFloat(
     Vector=<0.05 0.01 0.0075>
     Float=&Multiply_Float(
      A=&RandomRangeFloat(Minimum=1.5, Maximum=5)
      B=&FN_FrontEndFloatModifier(Gameplay=1, Front End=$User.FrontEndScale)
     )
    )
    Mesh Scale Mode=::NewEnumerator4
    Sprite Size Max=<2 6>
    Sprite Size Min=<1.5 2.5>
    Sprite Size Mode=::NewEnumerator2
    UsePositionOffset=false
    Write Material Random=true
   )
   &ShapeLocation(
    MaskSpawnGroup=false
    Override Seed=false
    Sphere Radius=&MaskFloatBySpawnGroup(Masked Float=0, Passthrough Float=30)
    UseOwnerScale=true
   )
   &AddVelocityFromPoint(
    Module.Use Falloff Distance=false
    Velocity Strength=&Multiply_Float(
     A=&RandomRangeFloat(Minimum=205, Maximum=400)
     B=&FN_FrontEndFloatModifier(Gameplay=1, Front End=$User.FrontEndScale)
    )
   )
   &AddVelocity(
    Velocity=&RandomRangeVector(Minimum=<100 0 0>, Maximum=<200 0 0>)
    Velocity Speed Scale=&FN_FrontEndFloatModifier(Gameplay=1, Front End=0)
   )
   /SetVariables(Particles.VisibilityTag=&SpawnGroup()) {isSetParameters=true}
  ]
  particleUpdateScript [
   &ParticleState()
   &ScaleMeshSize(Scale Factor=&Multiply_VectorByFloat(
     Vector=&MakeVector(
      X=&FloatFromCurve(FloatCurve=DataInterface() {propertyValues {Curve {keys [{arriveTangent=-1.6549, interpMode=RCIM_Cubic, leaveTangent=-1.6549, tangentMode=RCTM_User, time=0, value=1.375}, <at 1 0 RCIM_Cubic>]}}})
      Y=1
      Z=1
     )
     Float=&Multiply_Float(
      A=&RandomRangeFloat(Minimum=0.75, Maximum=1)
      B=&FN_FrontEndFloatModifier(Gameplay=1, Front End=$User.FrontEndScale)
     )
    ))
   &ScaleSpriteSize(
    Uniform Curve Scale=&FN_FrontEndFloatModifier(Gameplay=1.3, Front End=0.7)
    Uniform Curve Sprite Scale=DataInterface() {propertyValues {Curve {keys [
        <at 0 0 RCIM_Cubic>
        {interpMode=RCIM_Cubic, tangentMode=RCTM_User, time=0.118, value=1}
        {interpMode=RCIM_Cubic, tangentMode=RCTM_User, time=1, value=0}
       ]}}}
   )
   &ScaleColor(
    Scale Alpha=&FloatFromCurve(FloatCurve=DataInterface() {propertyValues {Curve {keys [
         {arriveTangent=-0.0852, interpMode=RCIM_Cubic, leaveTangent=-0.0852, tangentMode=RCTM_User, time=0, value=1}
         {arriveTangent=-2.6621, interpMode=RCIM_Cubic, leaveTangent=-2.6621, tangentMode=RCTM_User, time=1, value=0}
        ]}}})
    Scale RGB=&MakeVectorFromLinearColorRGB(Color=&ColorFromCurve(ColorCurve=DataInterface() {propertyValues {
        AlphaCurve {keys [<at 0 1 RCIM_Cubic>]}
        BlueCurve {keys [<at 0 5 RCIM_Cubic>,<at 0.4538 1>,<at 0.6336 0 RCIM_Cubic>]}
        GreenCurve {keys [<at 0 3.1229 RCIM_Cubic>,<at 0.4538 0.2887>,<at 0.6336 0 RCIM_Cubic>]}
        RedCurve {keys [<at 0 0.1652 RCIM_Cubic>,<at 0.4538 0.033>,<at 0.6336 0 RCIM_Cubic>]}
       }}))
    ScaleA=true
    ScaleRGB=true
   )
   &Drag(
    Drag=&FloatFromCurve(FloatCurve=DataInterface() {propertyValues {Curve {keys [{arriveTangent=2.9839, interpMode=RCIM_Cubic, leaveTangent=2.9839, tangentMode=RCTM_User, time=0, value=1.875}, <at 1 6 RCIM_Cubic>]}}})
    Use Linear Drag=true
    Use Rotational Drag=true
   )
   &CurlNoiseForce(Bias Noise=false, Noise Strength=&RandomRangeFloat(Minimum=200, Maximum=400), Pan Noise=true)
   &AccelerationForce(Acceleration=<0 0 -500>)
   &SolveForcesAndVelocity(Clamp Velocity=false, Limit Acceleration=false)
   &ScaleSpriteSizeBySpeed(Max Scale Factor=<0.5 3>, Min Scale Factor=<1 2>, Velocity Threshold=500)
   /SetVariables(Particles.float:NiagaraFloat=&VectorLength(Vector=$Particles.Velocity)) {isSetParameters=true}
   /SetVariables(Particles.MeshColor:LinearColor=&MultiplyLinearColor(
     Linear Color=&MultiplyLinearColor(
      Linear Color=&ColorFromCurve(
       ColorCurve=DataInterface() {propertyValues {
         AlphaCurve {keys [<at 0 1 RCIM_Cubic>]}
         BlueCurve {keys [<at 0.025 5>, {arriveTangent=-8.4211, interpMode=RCIM_Cubic, leaveTangent=-8.4211, time=0.3, value=2}, <at 0.5 1 RCIM_Cubic>]}
         GreenCurve {keys [<at 0.025 3.1719>, {arriveTangent=-6.392, interpMode=RCIM_Cubic, leaveTangent=-6.392, time=0.3, value=1.6436}, <at 0.5 0.1356 RCIM_Cubic>]}
         RedCurve {keys [<at 0.025 0.2>, {arriveTangent=1.055, interpMode=RCIM_Cubic, leaveTangent=1.055, time=0.3, value=0.88}, <at 0.5 0.7011 RCIM_Cubic>]}
        }}
       CurveIndex=&RandomRangeFloat(Minimum=0, Maximum=1)
      )
      Scale Factor=/Niagara/DynamicInputs/Multiply/Multiply_Vector4_ByFloat(
       Vector4:Vector4f=&MakeVector4_FromLinearColor(LinearColor=&ColorFromCurve(ColorCurve=DataInterface() {propertyValues {
           AlphaCurve {keys [<at 0 1 RCIM_Cubic>]}
           BlueCurve {keys [{arriveTangent=-4.2105, interpMode=RCIM_Cubic, time=0, value=1}, <at 0.5 1 RCIM_Cubic>]}
           GreenCurve {keys [{arriveTangent=-3.3388, interpMode=RCIM_Cubic, time=0, value=1}, <at 0.5 0.3896 RCIM_Cubic>]}
           RedCurve {keys [{arriveTangent=0.4778, interpMode=RCIM_Cubic, time=0, value=1}, <at 0.5 0 RCIM_Cubic>]}
          }}))
       Float:NiagaraFloat=&FNTimeOfDayScalar(Night Scale=1.3, Morning Scale=1.8, Evening Scale=1.8)
      )
     )
     Scale Factor=&MakeVector4FromVectorAndFloat(
      "Vector (RGB)"=<1 1 1>
      "Float (Alpha)"=&FloatFromCurve(FloatCurve=DataInterface() {propertyValues {Curve {keys [<at 0 1 RCIM_Cubic>,<at 0.5 0>]}}})
     )
    )) {isSetParameters=true}
  ]
  rendererClasses [&NiagaraSpriteRendererProperties, &NiagaraMeshRendererProperties]
  renderers [
   {rendererClass=&NiagaraSpriteRendererProperties}
   {rendererClass=&NiagaraMeshRendererProperties, rendererIndex=1}
  ]
 }
}
summary {
 systemName=NS_Guitar_FareSpore_Impact
 emitters [
  {emitterName=Glows, rendererClasses [&NiagaraSpriteRendererProperties, &NiagaraLightRendererProperties]}
  {emitterName=Impact_Flare, rendererClasses [&NiagaraMeshRendererProperties]}
  {
   emitterName=FN_PhysMatDebris
   rendererClasses [&NiagaraSpriteRendererProperties, &NiagaraMeshRendererProperties]
  }
  {emitterName=Spark, rendererClasses [&NiagaraSpriteRendererProperties, &NiagaraMeshRendererProperties]}
  {emitterName=Smoke, rendererClasses [&NiagaraSpriteRendererProperties]}
  {emitterName=Mesh_Box, rendererClasses [&NiagaraMeshRendererProperties, &NiagaraSpriteRendererProperties]}
 ]
 userVariables [
  User.bIsFrontEnd {type=NiagaraBool}
  User.bIsFrontEndPreview {type=NiagaraBool}
  User.FrontEndScale {type=NiagaraFloat, defaultValue=0.75, description=""}
  User.PhysicalSurface {type=NiagaraInt32}
 ]
}
